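# Translucent unlit, two-sided spline-mesh material for asteroid dust ribbons in the Asteria lobby's space environment.
# Panned dust and smoke-trail textures build opacity, shaped by depth fade, view-angle fade and optional end fade.
# Emissive fakes lighting by dotting combined dust and smoke normals with a parameter light direction, over a fixed or time-of-day color.
# SplineThicken, tangent-offset math and a sphere normal sample are left unconnected.
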Material /Game/Environments/Asteria/Props/LobbyIslandSpace/ClusterSpline/M_Armadillo_AsteroidDust
graph {
 properties {
  bTangentSpaceNormal=false
  bUseMaterialAttributes=true
  bUsedWithSplineMeshes=true
  blendMode=BLEND_Translucent
  pixelDepthOffsetMode=PDOM_Legacy
  shadingModel=MSM_Unlit
  twoSided=true
  usedShadingModels=Unlit
 }
 nodes [
  n1 = ScalarParameter() {defaultValue=5000, parameterName=SplineWidth}
  MaterialFunctionCall(WidthBase=@n1, WidthTip=@n1) {materialFunction=&SplineThicken}
  n2 = ComponentMask(None=TextureCoordinate()) {r=true}
  Multiply(
   A=Multiply(A=Subtract(A=@n2) {constB=0.5}, B=MaterialFunctionCall() {materialFunction=&TangentBasis}.X U) {constB=2}
   B=ScalarParameter()
  )
  n3 = TextureCoordinate()
  n4 = VectorParameter() {parameterName=Tiles, defaultValue=<rgba 2 3 3 16>}
  Multiply(A=@n3, B=AppendVector(A=@n4.R, B=@n4.R))
  n5 = ScalarParameter() {defaultValue=1, parameterName=TileMult}
  n6 = Multiply(A=@n4.A, B=@n5)
  n7 = VectorParameter() {parameterName=PanSpeeds, defaultValue=<rgba 0.01 0.05 0.0025 0>}
  n8 = Add(A=Time(), B=ScalarParameter() {parameterName=Random})
  n9 = Panner(Coordinate=Multiply(A=@n3, B=AppendVector(A=@n4.B, B=@n6)), Time=@n8, Speed=AppendVector(A=@n7.R, B=@n7.R))
  n10 = Multiply(A=TextureSample(UVs=Multiply(A=@n9) {constB=-1.4}) {texture=/Game/Global/Textures/Default/PointSplatter/T_Tiling_Dust}.RGB) {constB=3}
  n11 = AppendVector(A=Constant() {r=1}, B=Multiply(A=@n5, B=ScalarParameter() {defaultValue=2, parameterName=RibbonTile}))
  n12 = TextureCoordinate()
  n13 = Panner(
   Coordinate=MaterialFunctionCall(XY=Multiply(A=@n11, B=@n12)) {materialFunction=&Swizzle}.YX
   Time=Multiply(A=@n7.B, B=@n8)
  ) {speedX=1}
  n14 = TextureSample(UVs=@n13) {samplerType=SAMPLERTYPE_Grayscale, texture=&T_Smoke_Trail_Soft}
  n15 = Add(
   A=TextureSample(UVs=@n9) {samplerType=SAMPLERTYPE_LinearGrayscale, texture=/Game/Effects/Fort_Effects/Textures/Rocks/T_Tiling_Dust_Dirt_Bits_M}.G
   B=Multiply(A=Add(A=@n10) {constB=0.25}, B=@n14.G)
  )
  n16 = Multiply(A=ScalarParameter() {defaultValue=4, parameterName=Opacity}, B=Multiply(A=@n15, B=@n14.G))
  n17 = ScalarParameter() {defaultValue=1000, parameterName=DepthFade}
  n18 = Multiply(
   A=QualitySwitch(Default=DepthFade(Opacity=@n16, FadeDistance=@n17), Low=@n16)
   B=MaterialFunctionCall("FadeSensitivity (optional)"=ScalarParameter() {defaultValue=1, parameterName=AngleFade}) {materialFunction=&FadeBasedOnViewAngle}.Result
  )
  n19 = ComponentMask(None=TextureCoordinate()) {g=true}
  n20 = OneMinus(None=@n19)
  n21 = Saturate(None=LinearInterpolate(
    A=ScalarParameter() {defaultValue=-0.2912, parameterName=EndFade_Low}
    B=ScalarParameter() {defaultValue=16.0885, parameterName=EndFade_High}
    Alpha=Multiply(A=@n19, B=@n20)
   ))
  n22 = StaticSwitchParameter(
   True=MaterialFunctionCall(
    Color_Morning=VectorParameter() {parameterName=CloudColor_Morning, defaultValue=<rgba 0.0629 0.0376 0.099 1>}.RGB
    Color_Day=VectorParameter() {parameterName=CloudColor_Day, defaultValue=<rgba 0.0625 0.0625 0.0625 1>}.RGB
    Color_Evening=VectorParameter() {parameterName=CloudColor_Evening, defaultValue=<rgba 0.0521 0.0122 0.0222 1>}.RGB
    Color_Night=VectorParameter() {parameterName=CloudColor_Night, defaultValue=<rgba 0.0067 0.0083 0.0208 1>}.RGB
   ) {materialFunction=&Color_Time_of_Day}.Result
   False=VectorParameter() {parameterName=Color, defaultValue=<rgba 0.02 0.0171 0.0199 1>}.RGB
  ) {parameterName=UseTODMColors}
  n23 = Multiply(A=@n22, B=@n15)
  n24 = TextureSample(UVs=@n13) {samplerType=SAMPLERTYPE_Normal, texture=/Game/Effects/Fort_Effects/Textures/Smoke/T_Smoke_Trail_Soft_N}
  n25 = Constant3Vector() {constant=<rgba 1 1 0 0>}
  n26 = Normalize(VectorInput=Add(
    A=Multiply(A=@n24.RGB, B=@n25)
    B=TextureSample(UVs=@n9) {samplerType=SAMPLERTYPE_Normal, texture=/Game/Effects/Fort_Effects/Textures/Rocks/T_Tiling_Dust_Dirt_Bits_N}.RGB
   ))
  n27 = VectorParameter() {parameterName=LightDirection, defaultValue=<rgba 1 0 -0.7 1>}
  n28 = LinearInterpolate(
   B=VectorParameter() {parameterName=LightColor, defaultValue=<rgba 0.2656 0.6083 1 0>}.RGB
   Alpha=DotProduct(A=Transform(None=@n26), B=Normalize(VectorInput=@n27.RGB))
  ) {constA=0.24}
  SetMaterialAttributes!(
   MaterialAttributes
   Opacity=StaticSwitchParameter(True=Multiply(A=@n18, B=@n21), False=@n18) {parameterName=UseEndFade}
   Emissive Color=Add(A=@n22, B=Multiply(A=@n23, B=@n28))
  ) {attributeSetTypes [#1, #2]}
  TextureSample(UVs=TextureCoordinate() {uTiling=5, vTiling=40}) {samplerType=SAMPLERTYPE_Normal, texture=&T_Sphere_N}
 ]
}
